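# Physics asset for the OceanBreeze medium male body's skirt during skydiving, with capsules on skirt bones, pelvis, spine and thighs.
# Six four-bone chains (A, B, C; left and right) hang from spine_02 with free angular motion at their roots and tight twist limits along each chain.
# Extra constraints cross-link neighboring chains, join the C tips left to right, and tie A02 and B02 back to spine_02.
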PhysicsAsset /Game/Characters/Player/Male/Medium/Bodies/M_MED_OceanBreeze/Meshes/M_MED_OceanBreeze_Skydiving_Physics
bodies {
 dyn_SkirtA00_l [{shapeType=Capsule, length=4, radius=6.5, shapeName=dyn_SkirtB00_l_capsule, center=<0 1.446 0>}]
 dyn_SkirtA00_r [{shapeType=Capsule, length=4, radius=6.5, shapeName=dyn_SkirtB00_r_capsule, center=<0 1.446 0>}]
 dyn_SkirtA01_l [{shapeType=Capsule, length=4, radius=6.5, shapeName=dyn_SkirtB01_l_capsule, center=<4.89 2.6607 -0.3255>, rotation=<rot 0 0 15>}]
 dyn_SkirtA01_r [{shapeType=Capsule, length=6.5, radius=6.75, shapeName=dyn_SkirtB01_r_capsule, center=<4.89 4.047 -0.2722>, rotation=<rot 0 0 -20>}]
 dyn_SkirtA02_l [{shapeType=Capsule, length=4, radius=6.5, shapeName=dyn_SkirtB02_l_capsule, center=<4.89 4.2446 -0.7499>, rotation=<rot 0 0 15>}]
 dyn_SkirtA02_r [{shapeType=Capsule, length=6, radius=7.25, shapeName=dyn_SkirtB02_r_capsule, center=<1.9303 3.587 -0.4396>, rotation=<rot 0 0 -20>}]
 dyn_SkirtA03_l [{shapeType=Capsule, length=9, radius=6.25, shapeName=dyn_SkirtB03_l_capsule, center=<4.89 4.1438 0.2709>, rotation=<rot -0.0001 0.0001 15>}]
 dyn_SkirtA03_r [{shapeType=Capsule, length=6.25, radius=7, shapeName=dyn_SkirtB03_r_capsule, center=<2.2199 3.8632 -0.3391>, rotation=<rot 0 0 -20>}]
 dyn_SkirtB00_l [{shapeType=Capsule, length=4, radius=6.5, shapeName=dyn_SkirtB00_l_capsule, center=<0 1.446 0>}]
 dyn_SkirtB00_r [{shapeType=Capsule, length=4, radius=6.5, shapeName=dyn_SkirtB00_r_capsule, center=<0 1.446 0>}]
 dyn_SkirtB01_l [{shapeType=Capsule, length=4, radius=6.5, shapeName=dyn_SkirtB01_l_capsule, center=<3.9117 2.9344 0>}]
 dyn_SkirtB01_r [{shapeType=Capsule, length=8.5, radius=6.5, shapeName=dyn_SkirtB01_r_capsule, center=<3.9117 2.515 0.676>}]
 dyn_SkirtB02_l [{shapeType=Capsule, length=11.25, radius=6.5, shapeName=dyn_SkirtB02_l_capsule, center=<2.49 5.7619 2.0636>}]
 dyn_SkirtB02_r [{shapeType=Capsule, length=9.25, radius=5.75, shapeName=dyn_SkirtB02_r_capsule, center=<3.9117 4.1835 0.676>}]
 dyn_SkirtB03_l [{shapeType=Capsule, length=10.25, radius=6.25, shapeName=dyn_SkirtB03_l_capsule, center=<4.9903 6.5575 2.3891>, rotation=<rot -0.0001 0 -10>}]
 dyn_SkirtB03_r [{shapeType=Capsule, length=11.5, radius=5.75, shapeName=dyn_SkirtB03_r_capsule, center=<5.3822 4.1835 0.676>}]
 dyn_SkirtC00_l [{shapeType=Capsule, length=4, radius=6.5, shapeName=dyn_SkirtC00_l_capsule, center=<0 1.446 0>}]
 dyn_SkirtC00_r [{shapeType=Capsule, length=4, radius=6.5, shapeName=dyn_SkirtC00_r_capsule, center=<0 1.446 0>}]
 dyn_SkirtC01_l [{shapeType=Capsule, length=9.75, radius=6.5, shapeName=dyn_SkirtC01_l_capsule, center=<3.9117 3.1046 0>}]
 dyn_SkirtC01_r [{shapeType=Capsule, length=9.75, radius=6, shapeName=dyn_SkirtC01_r_capsule, center=<3.9117 2.058 0.5006>}]
 dyn_SkirtC02_l [{shapeType=Capsule, length=13.5, radius=6.5, shapeName=dyn_SkirtC02_l_capsule, center=<2.6892 4.2542 0>}]
 dyn_SkirtC02_r [{shapeType=Capsule, length=9.75, radius=6, shapeName=dyn_SkirtC02_r_capsule, center=<2.6892 4.5745 0.5006>}]
 dyn_SkirtC03_l [{shapeType=Capsule, length=11.25, radius=7.25, shapeName=dyn_SkirtC03_l_capsule, center=<1.4842 5.11 -1.2194>, rotation=<rot 0 0 10>}]
 dyn_SkirtC03_r [{shapeType=Capsule, length=8.25, radius=6.5, shapeName=dyn_SkirtC03_r_capsule, center=<2.4345 6.1676 0.7772>, rotation=<rot 0 -0.0001 -5>}]
 pelvis [
  {shapeType=Capsule, length=11.25, radius=11, shapeName=pelvis_capsule, center=<0.1982 -1.2398 0>}
  {shapeType=Capsule, length=15, radius=9, shapeName=pelvis_capsule, center=<-7.9326 -4.2175 0>}
 ]
 spine_01 [{shapeType=Capsule, length=8.75, radius=8.5, shapeName=spine_01_capsule, center=<3.1065 0 0>}]
 spine_02 [{shapeType=Capsule, length=6.5, radius=9.75, shapeName=spine_02_capsule, center=<1.3835 0 0>}]
 spine_03 [{shapeType=Capsule, length=7, radius=9.25, shapeName=spine_03_capsule}]
 thigh_l [{
   shapeType=Capsule
   length=30.9009
   radius=7.5521
   shapeName=thigh_l_capsule
   center=<-22.1669 -0.2135 0.8934>
   rotation=<rot -18.1895 86.4286 -90.8175>
  }]
 thigh_r [{
   shapeType=Capsule
   length=31.2287
   radius=7.6301
   shapeName=thigh_r_capsule
   center=<22.4647 -0.8 -0.6954>
   rotation=<rot -7.5491 91.0332 -92.3847>
  }]
}
constraints [
 {bone1Name=spine_01, bone2Name=pelvis, swing1LimitDegrees=45, swing2LimitDegrees=45, twistLimitDegrees=45}
 {bone1Name=spine_02, bone2Name=spine_01, swing1LimitDegrees=45, swing2LimitDegrees=45, twistLimitDegrees=45}
 {
  bone1Name=thigh_l
  bone2Name=pelvis
  swing1LimitDegrees=45
  swing1Motion=Free
  swing2LimitDegrees=45
  swing2Motion=Free
  twistLimitDegrees=45
  twistMotion=Free
 }
 {
  bone1Name=thigh_r
  bone2Name=pelvis
  swing1LimitDegrees=45
  swing1Motion=Free
  swing2LimitDegrees=45
  swing2Motion=Free
  twistLimitDegrees=45
  twistMotion=Free
 }
 {
  bone1Name=dyn_SkirtB00_l
  bone2Name=spine_02
  swing1LimitDegrees=45
  swing1Motion=Free
  swing2LimitDegrees=45
  swing2Motion=Free
  twistLimitDegrees=45
  twistMotion=Free
 }
 {bone1Name=dyn_SkirtB01_l, bone2Name=dyn_SkirtB00_l, swing1LimitDegrees=45, swing2LimitDegrees=5, twistLimitDegrees=1}
 {bone1Name=dyn_SkirtB02_l, bone2Name=dyn_SkirtB01_l, swing1LimitDegrees=45, swing2LimitDegrees=5, twistLimitDegrees=1}
 {bone1Name=dyn_SkirtB03_l, bone2Name=dyn_SkirtB02_l, swing1LimitDegrees=45, swing2LimitDegrees=5, twistLimitDegrees=1}
 {
  bone1Name=dyn_SkirtB00_r
  bone2Name=spine_02
  swing1LimitDegrees=45
  swing1Motion=Free
  swing2LimitDegrees=45
  swing2Motion=Free
  twistLimitDegrees=45
  twistMotion=Free
 }
 {bone1Name=dyn_SkirtB01_r, bone2Name=dyn_SkirtB00_r, swing1LimitDegrees=45, swing2LimitDegrees=5, twistLimitDegrees=1}
 {bone1Name=dyn_SkirtB02_r, bone2Name=dyn_SkirtB01_r, swing1LimitDegrees=45, swing2LimitDegrees=5, twistLimitDegrees=1}
 {bone1Name=dyn_SkirtB03_r, bone2Name=dyn_SkirtB02_r, swing1LimitDegrees=45, swing2LimitDegrees=5, twistLimitDegrees=1}
 {
  bone1Name=dyn_SkirtC00_l
  bone2Name=spine_02
  swing1LimitDegrees=45
  swing1Motion=Free
  swing2LimitDegrees=45
  swing2Motion=Free
  twistLimitDegrees=45
  twistMotion=Free
 }
 {bone1Name=dyn_SkirtC01_l, bone2Name=dyn_SkirtC00_l, swing1LimitDegrees=45, swing2LimitDegrees=5, twistLimitDegrees=1}
 {bone1Name=dyn_SkirtC02_l, bone2Name=dyn_SkirtC01_l, swing1LimitDegrees=45, swing2LimitDegrees=5, twistLimitDegrees=1}
 {
  bone1Name=dyn_SkirtC00_r
  bone2Name=spine_02
  swing1LimitDegrees=45
  swing1Motion=Free
  swing2LimitDegrees=45
  swing2Motion=Free
  twistLimitDegrees=45
  twistMotion=Free
 }
 {bone1Name=dyn_SkirtC01_r, bone2Name=dyn_SkirtC00_r, swing1LimitDegrees=45, swing2LimitDegrees=5, twistLimitDegrees=1}
 {bone1Name=dyn_SkirtC02_r, bone2Name=dyn_SkirtC01_r, swing1LimitDegrees=45, swing2LimitDegrees=5, twistLimitDegrees=1}
 {
  bone1Name=dyn_SkirtA00_l
  bone2Name=spine_02
  swing1LimitDegrees=45
  swing1Motion=Free
  swing2LimitDegrees=45
  swing2Motion=Free
  twistLimitDegrees=45
  twistMotion=Free
 }
 {bone1Name=dyn_SkirtA01_l, bone2Name=dyn_SkirtA00_l, swing1LimitDegrees=45, swing2LimitDegrees=5, twistLimitDegrees=1}
 {bone1Name=dyn_SkirtA02_l, bone2Name=dyn_SkirtA01_l, swing1LimitDegrees=45, swing2LimitDegrees=5, twistLimitDegrees=1}
 {bone1Name=dyn_SkirtA03_l, bone2Name=dyn_SkirtA02_l, swing1LimitDegrees=45, swing2LimitDegrees=5, twistLimitDegrees=1}
 {
  bone1Name=dyn_SkirtA00_r
  bone2Name=spine_02
  swing1LimitDegrees=45
  swing1Motion=Free
  swing2LimitDegrees=45
  swing2Motion=Free
  twistLimitDegrees=45
  twistMotion=Free
 }
 {bone1Name=dyn_SkirtA01_r, bone2Name=dyn_SkirtA00_r, swing1LimitDegrees=45, swing2LimitDegrees=5, twistLimitDegrees=1}
 {bone1Name=dyn_SkirtA02_r, bone2Name=dyn_SkirtA01_r, swing1LimitDegrees=45, swing2LimitDegrees=5, twistLimitDegrees=1}
 {bone1Name=dyn_SkirtA03_r, bone2Name=dyn_SkirtA02_r, swing1LimitDegrees=45, swing2LimitDegrees=5, twistLimitDegrees=1}
 {
  bone1Name=dyn_SkirtB02_r
  bone2Name=spine_02
  swing1LimitDegrees=45
  swing1Motion=Free
  swing2LimitDegrees=45
  swing2Motion=Free
  twistLimitDegrees=45
  twistMotion=Free
 }
 {
  bone1Name=dyn_SkirtB02_l
  bone2Name=spine_02
  swing1LimitDegrees=45
  swing1Motion=Free
  swing2LimitDegrees=45
  swing2Motion=Free
  twistLimitDegrees=45
  twistMotion=Free
 }
 {
  bone1Name=dyn_SkirtA02_r
  bone2Name=spine_02
  swing1LimitDegrees=45
  swing1Motion=Free
  swing2LimitDegrees=45
  swing2Motion=Free
  twistLimitDegrees=45
  twistMotion=Free
 }
 {
  bone1Name=dyn_SkirtA02_l
  bone2Name=spine_02
  swing1LimitDegrees=45
  swing1Motion=Free
  swing2LimitDegrees=45
  swing2Motion=Free
  twistLimitDegrees=45
  twistMotion=Free
 }
 {bone1Name=dyn_SkirtC03_l, bone2Name=dyn_SkirtC02_l, swing1LimitDegrees=45, swing2LimitDegrees=5, twistLimitDegrees=1}
 {bone1Name=dyn_SkirtC03_r, bone2Name=dyn_SkirtC02_r, swing1LimitDegrees=45, swing2LimitDegrees=5, twistLimitDegrees=1}
 {
  bone1Name=dyn_SkirtC02_r
  bone2Name=dyn_SkirtB02_r
  swing1LimitDegrees=45
  swing1Motion=Free
  swing2LimitDegrees=45
  swing2Motion=Free
  twistLimitDegrees=45
  twistMotion=Free
 }
 {
  bone1Name=dyn_SkirtC03_l
  bone2Name=dyn_SkirtC03_r
  swing1LimitDegrees=45
  swing1Motion=Free
  swing2LimitDegrees=45
  swing2Motion=Free
  twistLimitDegrees=45
  twistMotion=Free
 }
 {
  bone1Name=dyn_SkirtC02_l
  bone2Name=dyn_SkirtB02_l
  swing1LimitDegrees=45
  swing1Motion=Free
  swing2LimitDegrees=45
  swing2Motion=Free
  twistLimitDegrees=45
  twistMotion=Free
 }
 {bone1Name=dyn_SkirtA03_l, bone2Name=dyn_SkirtB03_l, swing1LimitDegrees=45, swing2LimitDegrees=45, twistLimitDegrees=45}
 {bone1Name=dyn_SkirtA03_r, bone2Name=dyn_SkirtB03_r, swing1LimitDegrees=45, swing2LimitDegrees=45, twistLimitDegrees=45}
 {
  bone1Name=dyn_SkirtA01_r
  bone2Name=dyn_SkirtB01_r
  swing1LimitDegrees=45
  swing1Motion=Free
  swing2LimitDegrees=45
  swing2Motion=Free
  twistLimitDegrees=45
  twistMotion=Free
 }
 {
  bone1Name=dyn_SkirtC03_r
  bone2Name=dyn_SkirtB03_r
  swing1LimitDegrees=45
  swing1Motion=Free
  swing2LimitDegrees=45
  swing2Motion=Free
  twistLimitDegrees=45
  twistMotion=Free
 }
 {
  bone1Name=dyn_SkirtC03_l
  bone2Name=dyn_SkirtB03_l
  swing1LimitDegrees=45
  swing1Motion=Free
  swing2LimitDegrees=45
  swing2Motion=Free
  twistLimitDegrees=45
  twistMotion=Free
 }
]
